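# Translucent unlit, two-sided Niagara sprite material that draws an animated face for the SpaceFeline pickaxe.
# Two half-scale samples of one face texture supply eyes (R) and mouth (G), offset by dynamic parameters EyesU and MouthU on both axes, leaving EyesV and MouthV unused.
# A time-scrolled vertical sine creates scanlines that brighten the particle color and modulate the combined mask, which is also multiplied by particle alpha.
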Material /BRCosmetics/Weapons/FORT_Melee/Pickaxe_SpaceFeline/FX/M_Pickaxe_SpaceFeline_Face
graph {
 properties {
  bUsedWithNiagaraSprites=true
  blendMode=BLEND_Translucent
  pixelDepthOffsetMode=PDOM_Legacy
  shadingModel=MSM_Unlit
  twoSided=true
  usedShadingModels=Unlit
 }
 nodes [
  n1 = ParticleColor()
  n2 = ComponentMask(None=TextureCoordinate()) {g=true}
  n3 = Multiply(A=Time()) {constB=0.05}
  n4 = Sine(None=Multiply(A=Add(A=@n2, B=@n3)) {constB=50})
  MaterialFunctionCall(Emissive=LinearInterpolate(A=@n1.RGB, B=Multiply(A=@n1.RGB) {constB=3}, Alpha=Saturate(None=@n4))) {materialFunction=&MF_FX_EmissiveAdjustments}
  n5 = Multiply(A=TextureCoordinate()) {constB=0.5}
  n6 = DynamicParameter() {paramNames [EyesU,EyesV,MouthU,MouthV]}
  n7 = AppendVector(A=@n6.EyesU, B=@n6.EyesU)
  n8 = Multiply(A=TextureCoordinate()) {constB=0.5}
  n9 = AppendVector(A=@n6.MouthU, B=@n6.MouthU)
  n10 = Add(
   A=TextureSample(UVs=Add(A=@n5, B=@n7)) {texture=&/T_Pickaxe_SpaceFeline_Face}.R
   B=TextureSample(UVs=Add(A=@n8, B=@n9) {constB=0.5}) {texture=&T_Pickaxe_SpaceFeline_Face}.G
  )
  Multiply(A=@n1.A, B=@n10)
  n11 = Multiply(A=@n4, B=@n10)
  Saturate(None=Add(A=Saturate(None=@n11), B=Multiply(A=@n10) {constB=0.5}))
 ]
}
