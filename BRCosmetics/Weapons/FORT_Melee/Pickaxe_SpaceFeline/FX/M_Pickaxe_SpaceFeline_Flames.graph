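# Translucent, unlit, two-sided flame material for Niagara mesh particles in the SpaceFeline pickaxe effects.
# A spike texture panned along U is eroded at both ends by U gradients to form the flame shape.
# That shape blends border color toward a boosted core, tinted by particle color.
# UseColorOverlay adds a nebula texture projected in view space around the particle position before the emissive-adjustment function.
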Material /BRCosmetics/Weapons/FORT_Melee/Pickaxe_SpaceFeline/FX/M_Pickaxe_SpaceFeline_Flames
graph {
 properties {
  bUsedWithNiagaraMeshParticles=true
  blendMode=BLEND_Translucent
  pixelDepthOffsetMode=PDOM_Legacy
  shadingModel=MSM_Unlit
  twoSided=true
  usedShadingModels=Unlit
 }
 nodes [
  n1 = VectorParameter() {parameterName=CoreColor, defaultValue=<rgba 1 1 1 1>}
  n2 = VectorParameter() {parameterName=BorderColor, defaultValue=<rgba 1 0.044 0 1>}
  n3 = Multiply(A=@n2.RGB) {constB=10}
  n4 = ComponentMask(None=TextureCoordinate()) {r=true}
  n5 = Power(Base=@n4)
  n6 = Multiply(A=TextureCoordinate(), B=Constant2Vector() {g=1, r=-0.8})
  n7 = Multiply(A=Time(), B=Constant2Vector() {r=0.6})
  n8 = SmoothStep(Value=@n4) {constMax=2, constMin=0.4}
  n9 = Power(Base=SmoothStep(Value=OneMinus(None=@n4)) {constMin=0.75}) {constExponent=3}
  n10 = Subtract(
   A=TextureSample(UVs=Add(A=@n6, B=@n7)) {texture=/BRCosmetics/Weapons/FORT_Melee/Pickaxe_HornedJudgment/FX/T_Pickaxe_HornedJudgement_Spikes03}.R
   B=Saturate(None=Add(A=@n8, B=@n9))
  )
  n11 = Saturate(None=@n10)
  n12 = Power(Base=@n11) {constExponent=1.25}
  n13 = Multiply(A=LinearInterpolate(A=@n2.RGB, B=Multiply(A=@n1.RGB, B=@n3), Alpha=Multiply(A=@n5, B=@n12)), B=ParticleColor().RGB)
  n14 = Subtract(A=WorldPosition().XYZ, B=ParticlePositionWS())
  n15 = ComponentMask(None=Transform(None=Multiply(A=@n14) {constB=0.02}) {transformSourceType=TRANSFORMSOURCE_World, transformType=TRANSFORM_View}) {g=true, r=true}
  n16 = Multiply(A=Power(Base=TextureSample(UVs=@n15) {texture=/Game/Characters/Player/Male/Medium/Bodies/M_Med_Soldier_04/Skins/Galaxy/Textures/T_FN_Nebula}.RGB) {constExponent=1.2}) {constB=0.4}
  MaterialFunctionCall(Emissive=LinearInterpolate(A=@n13, B=Add(A=@n16, B=@n13), Alpha=ScalarParameter() {parameterName=UseColorOverlay})) {materialFunction=&MF_FX_EmissiveAdjustments}
  SmoothStep(Value=@n10) {constMax=0.5, constMin=0.1}
  SmoothStep(Value=@n11) {constMin=0.4}
 ]
}
