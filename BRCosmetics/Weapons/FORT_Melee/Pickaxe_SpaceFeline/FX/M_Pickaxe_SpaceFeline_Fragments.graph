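# Translucent unlit Niagara sprite material for the fragment effects of the Space Feline pickaxe.
# A grayscale fragment texture is multiplied by particle color and fed through the shared MF_FX_EmissiveAdjustments function for the glow.
# Particle alpha times the texture's red channel forms the mask for opacity, so the particle system controls both tint and fade.
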Material /BRCosmetics/Weapons/FORT_Melee/Pickaxe_SpaceFeline/FX/M_Pickaxe_SpaceFeline_Fragments
graph {
 properties {bUsedWithNiagaraSprites=true, blendMode=BLEND_Translucent, pixelDepthOffsetMode=PDOM_Legacy, shadingModel=MSM_Unlit, usedShadingModels=Unlit}
 nodes [
  n1 = TextureSample() {samplerType=SAMPLERTYPE_LinearGrayscale, texture=&/T_Pickaxe_SpaceFeline_Fragments}
  n2 = ParticleColor()
  MaterialFunctionCall(Emissive=Multiply(A=@n1.RGB, B=@n2.RGB)) {materialFunction=&MF_FX_EmissiveAdjustments}
  Multiply(A=@n2.A, B=@n1.R)
 ]
}
