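# Translucent unlit Niagara ribbon material for the Space Feline pickaxe trail.
# Scrolling sparkles, spike-edged bands and length fades form a mask that drives opacity and a three-color gradient, with pulsing lines from camera-space position.
# A per-particle-offset square SDF dissolves it through the dynamic Erosion parameter, adding a glowing rim.
# Depth fade softens default and deferred paths only, and a TrailBulge parameter scales a normal-based offset.
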Material /BRCosmetics/Weapons/FORT_Melee/Pickaxe_SpaceFeline/FX/M_Pickaxe_SpaceFeline_Trail
graph {
 properties {bUsedWithNiagaraRibbons=true, blendMode=BLEND_Translucent, pixelDepthOffsetMode=PDOM_Legacy, shadingModel=MSM_Unlit, usedShadingModels=Unlit}
 nodes [
  n1 = TextureCoordinate()
  n2 = Power(Base=OneMinus(None=ComponentMask(None=@n1) {r=true})) {constExponent=5}
  n3 = VectorParameter() {parameterName=ColorA, defaultValue=<rgba 0.0416 0.1036 0.2604 1>}
  n4 = VectorParameter() {parameterName=ColorB, defaultValue=<rgba 0.1912 0.5906 1 1>}
  n5 = VectorParameter() {parameterName=ColorC, defaultValue=<rgba 0.4104 0.6856 0.9115 1>}
  n6 = Multiply(A=TextureCoordinate() {coordinateIndex=1}, B=Constant2Vector() {g=1, r=1})
  n7 = Multiply(A=Time(), B=Constant2Vector() {g=-0.15, r=0.85})
  n8 = Multiply(A=TextureCoordinate() {coordinateIndex=1}, B=Constant2Vector() {g=1, r=1})
  n9 = Multiply(A=Time(), B=Constant2Vector() {g=0.1, r=0.5})
  n10 = Max(
   A=TextureSample(UVs=Add(A=@n6, B=@n7)) {texture=&/T_Pickaxe_SpaceFeline_Sparkles}.R
   B=TextureSample(UVs=Add(A=@n8, B=@n9)) {texture=&T_Pickaxe_SpaceFeline_Sparkles}.R
  )
  n11 = ComponentMask(None=TextureCoordinate()) {g=true}
  n12 = SmoothStep(Value=Sine(None=Multiply(A=@n11) {constB=0.5})) {constMax=0.55, constMin=0.5}
  n13 = Add(A=Min(A=Multiply(A=@n10) {constB=1.5}, B=Multiply(A=@n12) {constB=0.2}), B=Multiply(A=@n12) {constB=0.4})
  n14 = TextureCoordinate() {coordinateIndex=1}
  n15 = ComponentMask(None=@n14) {g=true}
  n16 = Multiply(A=Abs(None=Subtract(A=@n15) {constB=0.5})) {constB=2}
  n17 = AppendVector(A=Multiply(A=ComponentMask(None=@n14) {r=true}) {constB=-0.5}, B=Add(A=Saturate(None=@n16)) {constB=-0.61})
  n18 = ComponentMask(None=TextureCoordinate()) {g=true}
  n19 = SmoothStep(Value=@n18) {constMax=0.5, constMin=0.5}
  n20 = Multiply(
   A=SmoothStep(Value=TextureSample(UVs=@n17) {texture=/BRCosmetics/Weapons/FORT_Melee/Pickaxe_HornedJudgment/FX/T_Pickaxe_HornedJudgement_Spikes03}.R) {constMax=0.2, constMin=0.1}
   B=OneMinus(None=@n19)
  )
  n21 = Subtract(A=@n19, B=SmoothStep(Value=@n18) {constMax=0.92, constMin=0.9})
  n22 = Multiply(A=Subtract(A=Max(A=@n20, B=@n21), B=@n12)) {constB=2.5}
  n23 = OneMinus(None=ComponentMask(None=TextureCoordinate()) {r=true})
  n24 = ScalarParameter() {defaultValue=0.05, parameterName=TrailLeadingFade}
  n25 = Subtract(
   A=SmoothStep(Min=ScalarParameter() {defaultValue=0.1, parameterName=TrailLenght}, Value=@n23) {constMin=1}
   B=SmoothStep(Min=OneMinus(None=@n24), Value=@n23) {constMin=1}
  )
  n26 = Saturate(None=Multiply(A=Add(A=@n13, B=@n22), B=Multiply(A=@n25) {constB=2}))
  n27 = ComponentMask(None=TextureCoordinate()) {r=true}
  n28 = Multiply(A=@n26, B=SmoothStep(Value=OneMinus(None=@n27)) {constMin=0.12})
  n29 = Multiply(A=@n28, B=@n28)
  n30 = Multiply(A=@n5.RGB) {constB=2}
  n31 = TransformPosition(None=WorldPosition().XYZ) {transformSourceType=TRANSFORMPOSSOURCE_World, transformType=TRANSFORMPOSSOURCE_Camera}
  n32 = Add(
   A=Multiply(A=ComponentMask(None=@n31) {g=true}, B=ScalarParameter() {defaultValue=0.7, parameterName=Emissive Lines Tiling, sliderMax=1}) {constB=0.1}
   B=Multiply(A=Time()) {constB=2}
  )
  n33 = If(A=Subtract(A=Sine(None=@n32)) {constB=0.75}, "A > B"=Constant() {r=0.5}, "A < B"=Constant())
  n34 = Add(
   A=Multiply(A=@n2) {constB=0.3}
   B=LinearInterpolate(
    A=MaterialFunctionCall(A=@n3.RGB, B=@n4.RGB, C=@n5.RGB, Alpha=@n29) {materialFunction=&Lerp_3Color}.Result
    B=@n30
    Alpha=Multiply(A=@n2, B=@n33)
   )
  )
  n35 = Multiply(A=TextureCoordinate() {coordinateIndex=1}, B=Constant2Vector() {g=0.5, r=0.5})
  n36 = Multiply(A=ParticleRandom()) {constB=2}
  n37 = Add(
   A=TextureSample(UVs=Add(A=@n35, B=@n36)) {texture=/BRCosmetics/Weapons/FORT_Melee/Pickaxe_CatSpace/FX/T_SquareSDF_02}.G
   B=LinearInterpolate(Alpha=DynamicParameter() {defaultValue=<rgba 0 1 1 1>, paramNames [Erosion,Param2,Param3,Param4]}.Erosion) {constA=-1}
  )
  n38 = SmoothStep(Value=@n37) {constMax=0.25, constMin=0.2}
  n39 = Multiply(A=Subtract(A=SmoothStep(Value=@n37) {constMax=0.15, constMin=0.1}, B=@n38), B=@n30)
  n40 = ParticleColor()
  MaterialFunctionCall(Emissive=Multiply(A=Add(A=@n34, B=@n39), B=@n40.RGB)) {materialFunction=&MF_FX_EmissiveAdjustments}
  n41 = Multiply(A=ComponentMask(None=TextureCoordinate()) {g=true}) {constB=0.5}
  Multiply(A=Power(Base=Sine(None=@n41)), B=Multiply(A=VertexNormalWS(), B=ScalarParameter() {parameterName=TrailBulge}) {constB=-5})
  n42 = Subtract(A=Multiply(A=@n26, B=@n40.A), B=Saturate(None=@n38))
  n43 = DepthFade(Opacity=@n42) {fadeDistanceDefault=15}
  ShadingPathSwitch(Default=@n43, Deferred=@n43, Forward=@n42, Mobile=@n42)
  Time() {bOverride_Period=true, period=1}
 ]
}
