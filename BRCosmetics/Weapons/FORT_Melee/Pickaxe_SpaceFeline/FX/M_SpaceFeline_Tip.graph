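# Unlit material for the Space Feline pickaxe tip effect, built for Niagara mesh particles.
# A two-color UV gradient shifts between boosted and base brightness using vertex-color green minus two panning glitch-noise samples, tinted by particle color.
# A nebula texture, projected in view space around the particle position, adds through UseColorOverlay, off by default, before emissive adjustments.
# A WPO branch scales vertex normals by panning blurred noise.
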Material /BRCosmetics/Weapons/FORT_Melee/Pickaxe_SpaceFeline/FX/M_SpaceFeline_Tip
graph {
 properties {
  bUsedWithNiagaraMeshParticles=true
  pixelDepthOffsetMode=PDOM_Legacy
  previewMesh=&/SM_Pickaxe_SpaceFeline_Tip
  shadingModel=MSM_Unlit
  usedShadingModels=Unlit
 }
 nodes [
  n1 = Power(Base=ComponentMask(None=TextureCoordinate()) {r=true}, Exp=ScalarParameter() {defaultValue=2, parameterName=GradientExponent})
  n2 = ScalarParameter() {defaultValue=5, parameterName=GradientMultiply}
  n3 = LinearInterpolate(
   A=VectorParameter() {parameterName=BorderColor, defaultValue=<rgba 0.1094 0.1094 0.1094 1>}.RGB
   B=VectorParameter() {parameterName=BorderColor2, defaultValue=<rgba 0 0 0 1>}.RGB
   Alpha=Saturate(None=Multiply(A=@n1, B=@n2))
  )
  n4 = Multiply(A=@n3) {constB=3}
  n5 = VertexColor()
  n6 = Multiply(A=TextureCoordinate(), B=Constant2Vector() {g=1, r=1})
  n7 = TextureSample(UVs=Panner(Coordinate=@n6) {speedX=-1}) {
   samplerType=SAMPLERTYPE_Masks
   texture=/Game/Environments/Asteria/Sets/NeonCity/Techart/Holograms/F_MED_BlueBadass_NeonLines_-_GlitchNoise_v1
  }
  n8 = TextureSample(UVs=Panner(Coordinate=@n6) {speedX=-0.5}) {
   samplerType=SAMPLERTYPE_Masks
   texture=&F_MED_BlueBadass_NeonLines_-_GlitchNoise_v1
  }
  n9 = Subtract(A=Multiply(A=Power(Base=@n5.G), B=ScalarParameter() {defaultValue=2, parameterName=MaskMultiplier}), B=Multiply(A=@n7.R, B=@n8.R))
  n10 = Multiply(A=LinearInterpolate(A=@n4, B=@n3, Alpha=Saturate(None=@n9)), B=ParticleColor().RGB)
  n11 = Subtract(A=WorldPosition().XYZ, B=ParticlePositionWS())
  n12 = ComponentMask(None=Transform(None=Multiply(A=@n11) {constB=0.02}) {transformSourceType=TRANSFORMSOURCE_World, transformType=TRANSFORM_View}) {g=true, r=true}
  n13 = Multiply(A=Power(Base=TextureSample(UVs=@n12) {texture=/Game/Characters/Player/Male/Medium/Bodies/M_Med_Soldier_04/Skins/Galaxy/Textures/T_FN_Nebula}.RGB) {constExponent=1.2}) {constB=0.6}
  MaterialFunctionCall(Emissive=LinearInterpolate(A=@n10, B=Add(A=@n13, B=@n10), Alpha=ScalarParameter() {parameterName=UseColorOverlay})) {materialFunction=&MF_FX_EmissiveAdjustments}
  Multiply(A=VectorParameter() {parameterName=CoreColor, defaultValue=<rgba 1 1 1 1>}.RGB, B=@n4)
  n14 = Panner(
   Coordinate=Multiply(A=TextureCoordinate(), B=ScalarParameter() {defaultValue=1.5, group=WPO, parameterName=WPO Noise Scale})
   Speed=AppendVector(
    A=ScalarParameter() {defaultValue=-0.5, group=WPO, parameterName=WPO Pan X}
    B=ScalarParameter() {defaultValue=-0.8, group=WPO, parameterName=WPO Pan Y}
   )
  )
  Multiply(
   A=Multiply(
    A=VertexNormalWS()
    B=TextureSample(UVs=@n14) {samplerType=SAMPLERTYPE_Grayscale, texture=&LowResBlurredNoise}.R
   )
   B=ScalarParameter() {defaultValue=0.4, group=WPO, parameterName=WPO Intensity}
  )
 ]
}
