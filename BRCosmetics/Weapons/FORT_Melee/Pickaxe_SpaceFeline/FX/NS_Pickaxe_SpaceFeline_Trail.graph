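# Niagara swing trail for the SpaceFeline pickaxe, layering two anim-trail ribbons and sprite squares on the weapon's skeletal mesh.
# FillTrail uses a system trail color, while pale HoloTrail darkens over particle life and animates a material parameter by emitter age.
# Squares spawn per distance moved, drift on curl noise and pick trail color A or B from VariantID.
# A matching OtherVars particle emitter is disabled.
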NiagaraSystem /BRCosmetics/Weapons/FORT_Melee/Pickaxe_SpaceFeline/FX/NS_Pickaxe_SpaceFeline_Trail
dependencies {
 usedDataInterfaces [&NiagaraDataInterfaceCurve, &NiagaraDataInterfaceVectorCurve, &NiagaraDataInterfaceArrayColor]
 usedDynamicInputs [&SelectLinearColorFromArray, &MakeIntFromFloat]
 usedModules [
  &self:SystemSpawnScript.SystemScriptSource.SystemScriptGraph.NiagaraNodeAssignment_0.SetVariables_#1
  &SystemState
 ]
}
emitters {
 FillTrail {
  emitterSpawnScript [&FN_RibbonIDOffsetUtility()]
  emitterUpdateScript [
   &EmitterState()
   &FN_SpawnAnimTrail()
  ]
  particleSpawnScript [
   &FN_InitializeAnimTrailBoneIndecies()
   &FN_InitializeAnimTrail(AnimTrailWidth=1.35, Color=$System.TrailColorA, Lifetime=0.4)
  ]
  particleUpdateScript [
   &SolveForcesAndVelocity(Clamp Velocity=false, Limit Acceleration=false)
   &FN_UpdateAnimTrail(Locked Particle RibbonWidth=1, SkeletalMesh=$User.SkeletalMesh) {module=FN_UpdateAnimTrail001}
   &ScaleRibbonWidth(Ribbon Width Scale=&FloatFromCurve(FloatCurve=DataInterface() {propertyValues {Curve {keys [
         {interpMode=RCIM_Cubic, leaveTangent=-2.8643, tangentMode=RCTM_Break, time=0, value=1.1}
         {arriveTangent=-0.0952, interpMode=RCIM_Cubic, tangentMode=RCTM_Break, time=0.6, value=0.35}
        ]}}}))
   &ScaleColor(
    Scale Alpha=&FloatFromCurve(
     FloatCurve=DataInterface() {propertyValues {Curve {keys [
         {interpMode=RCIM_Cubic, leaveTangent=-11.1311, tangentMode=RCTM_Break, time=0.15, value=1}
         {arriveTangent=-0.6732, tangentMode=RCTM_User, time=0.4, value=0}
        ]}}}
     Scale Curve=&FloatFromCurve(FloatCurve=DataInterface() {propertyValues {Curve {keys [<at 0 0>,<at 0.05 0.95>]}}}, CurveIndex=$Emitter.NormalizedLoopAge)
    )
    ScaleA=true
    ScaleRGB=true
   )
  ]
  rendererClasses [&NiagaraRibbonRendererProperties]
  renderers [{rendererClass=&NiagaraRibbonRendererProperties}]
 }
 HoloTrail {
  emitterSpawnScript [&FN_RibbonIDOffsetUtility()]
  emitterUpdateScript [
   &EmitterState()
   &FN_SpawnAnimTrail()
  ]
  particleSpawnScript [
   &FN_InitializeAnimTrailBoneIndecies()
   &FN_InitializeAnimTrail(Color=<rgba 0.779 0.9701 1 0.7>, Lifetime=0.4)
  ]
  particleUpdateScript [
   &SolveForcesAndVelocity(Clamp Velocity=false, Limit Acceleration=false)
   &FN_UpdateAnimTrail(Locked Particle RibbonWidth=1, SkeletalMesh=$User.SkeletalMesh) {module=FN_UpdateAnimTrail001}
   &ScaleRibbonWidth(Ribbon Width Scale=&FloatFromCurve(FloatCurve=DataInterface() {propertyValues {Curve {keys [
         {interpMode=RCIM_Cubic, leaveTangent=-2.8643, tangentMode=RCTM_Break, time=0, value=1.1}
         {arriveTangent=-0.0952, interpMode=RCIM_Cubic, tangentMode=RCTM_Break, time=0.6, value=0.35}
        ]}}}))
   &ScaleColor(
    Scale Alpha=&FloatFromCurve(
     FloatCurve=DataInterface() {propertyValues {Curve {keys [
         {interpMode=RCIM_Cubic, leaveTangent=-10.2145, tangentMode=RCTM_Break, time=0.3, value=0.95}
         {arriveTangent=-1.3272, tangentMode=RCTM_User, time=0.4, value=0.05}
        ]}}}
     Scale Curve=&FloatFromCurve(FloatCurve=DataInterface() {propertyValues {Curve {keys [<at 0 0>,<at 0.05 0.95>]}}}, CurveIndex=$Emitter.NormalizedLoopAge)
    )
    Scale RGB=&VectorFromCurve(VectorCurve=DataInterface() {propertyValues {
       XCurve {keys [
         {interpMode=RCIM_Cubic, leaveTangent=3.4171, tangentMode=RCTM_Break, time=0, value=1}
         {arriveTangent=-9.7574, arriveTangentWeight=0.0867, tangentMode=RCTM_User, time=0.26, value=0.2}
        ]}
       YCurve {keys [
         {interpMode=RCIM_Cubic, leaveTangent=3.4171, tangentMode=RCTM_Break, time=0, value=1}
         {arriveTangent=-9.7574, arriveTangentWeight=0.0867, tangentMode=RCTM_User, time=0.26, value=0.2}
        ]}
       ZCurve {keys [
         {interpMode=RCIM_Cubic, leaveTangent=3.4171, leaveTangentWeight=0.0867, tangentMode=RCTM_Break, time=0, value=1}
         {arriveTangent=-9.7574, tangentMode=RCTM_User, time=0.26, value=0.2}
        ]}
      }})
    ScaleA=true
    ScaleRGB=true
   )
   &DynamicMaterialParameters(
    Index 0 Param 1=&FloatFromCurve(
     FloatCurve=DataInterface() {propertyValues {Curve {keys [
         {arriveTangent=2.7331, interpMode=RCIM_Cubic, leaveTangent=1.4393, tangentMode=RCTM_Break, time=0.15, value=0}
         {arriveTangent=2.4667, interpMode=RCIM_Cubic, leaveTangent=4.5707, tangentMode=RCTM_Break, time=0.5, value=1}
        ]}}}
     CurveIndex=$Emitter.Age
    )
    Module.Index0Param0Write=true
    Module.Index0Param1Write=true
    Module.Index0Param2Write=true
    Module.Index0Param3Write=true
   )
  ]
  rendererClasses [&NiagaraRibbonRendererProperties]
  renderers [{rendererClass=&NiagaraRibbonRendererProperties}]
 }
 Squares_Var0and1_Only {
  emitterUpdateScript [
   &EmitterState()
   &SpawnPerUnit(Module.Use Max Movement Threshold=true, Module.Use Movement Tolerance=false, Use Spawn Probability=false)
  ]
  particleSpawnScript [
   &InitializeParticle(
    AdjustAlpha=false
    AdjustHue=true
    AdjustSaturation=false
    AdjustValue=false
    Color=&MakeCustomLinearColorFromBool(
     Bool=&SetBoolByFloatComparison(A=$User.VariantID, B=1, Comparison Type=::NewEnumerator6)
     True Color=$System.TrailColorB
     False Color=&HueShiftLinearColor(Hue=0, Saturation=1, Value=1.3, Linear Color=$System.TrailColorA)
    )
    Color Mode=::NewEnumerator3
    Hue Shift Range:Vector2f=<-0.05 0.05>
    Lifetime Max=0.9
    Lifetime Min=0.3
    Lifetime Mode=::NewEnumerator1
    Sprite Size Mode=::NewEnumerator3
    Uniform Sprite Size=&RandomRangeFloat(Minimum=2, Maximum=8)
    UsePositionOffset=false
    Write Material Random=true
   )
   &SkeletalMeshLocation(
    Bone / Socket Index=1
    MaskSpawnGroup=false
    Mesh Sampling Type=::NewEnumerator5
    Skeletal Mesh=$User.SkeletalMesh
    Socket Sampling Mode=::NewEnumerator5
   )
  ]
  particleUpdateScript [
   &ParticleState()
   &ScaleSpriteSize(
    Scale Sprite Size Mode=::NewEnumerator0
    Uniform Scale Factor=&Waveform(
     "Waveform [1]"=::NewEnumerator0
     "[1] Phase"=$Particles.NormalizedAge
     "[1] Frequency"=&RandomRangeFloat(Minimum=0.15, Maximum=2)
     "[1] Amplitude Min/Max"=<0 1>
    )
   )
   &CurlNoiseForce(Noise Frequency=15, Noise Strength=&RandomRangeFloat(Minimum=60, Maximum=250), Pan Noise=false)
   &InheritSourceMovement(Applied Force Scale=<0.1 0.1 0>, ApplyForce=true, ApplyForceFalloff=false, ApplyPosition=false, Module.LimitForceSpeed=false)
   &SolveForcesAndVelocity(Clamp Velocity=false, Limit Acceleration=false)
   &ScaleColor(
    Scale Alpha=&FloatFromCurve(
     FloatCurve=DataInterface() {propertyValues {Curve {keys [
         {arriveTangent=9.0725, interpMode=RCIM_Cubic, leaveTangent=9.0725, tangentMode=RCTM_User, time=0, value=0}
         {arriveTangent=-2.5138, interpMode=RCIM_Cubic, leaveTangent=-2.5138, tangentMode=RCTM_User, time=0.1408, value=2}
         {arriveTangent=-0.0405, interpMode=RCIM_Cubic, leaveTangent=-0.0405, tangentMode=RCTM_User, time=1, value=0}
        ]}}}
     Scale Curve=0.5
    )
    ScaleA=true
    ScaleRGB=true
   )
  ]
  rendererClasses [&NiagaraSpriteRendererProperties]
  renderers [{rendererClass=&NiagaraSpriteRendererProperties}]
 }
 TrailParticles_OtherVars {
  bEnabled=false
  emitterUpdateScript [
   &EmitterState()
   &SpawnPerUnit(Module.Use Max Movement Threshold=true, Module.Use Movement Tolerance=false, Use Spawn Probability=false)
  ]
  particleSpawnScript [
   &InitializeParticle(
    AdjustAlpha=false
    AdjustHue=true
    AdjustSaturation=false
    AdjustValue=false
    Color=&MakeCustomLinearColorFromBool(
     Bool=&SetBoolByFloatComparison(A=$User.VariantID, B=1, Comparison Type=::NewEnumerator6)
     True Color=$System.TrailColorB
     False Color=&HueShiftLinearColor(Hue=0, Saturation=1, Value=1.5, Linear Color=$System.TrailColorA)
    )
    Color Mode=::NewEnumerator3
    Hue Shift Range:Vector2f=<-0.05 0.05>
    Lifetime Max=0.9
    Lifetime Min=0.3
    Lifetime Mode=::NewEnumerator1
    Sprite Size Mode=::NewEnumerator3
    Uniform Sprite Size=&RandomRangeFloat(Minimum=2, Maximum=8)
    UsePositionOffset=false
    Write Material Random=true
   )
   &SkeletalMeshLocation(
    Bone / Socket Index=1
    MaskSpawnGroup=false
    Mesh Sampling Type=::NewEnumerator5
    Skeletal Mesh=$User.SkeletalMesh
    Socket Sampling Mode=::NewEnumerator5
   )
  ]
  particleUpdateScript [
   &ParticleState()
   &ScaleSpriteSize(
    Scale Sprite Size Mode=::NewEnumerator0
    Uniform Scale Factor=&Waveform(
     "Waveform [1]"=::NewEnumerator0
     "[1] Phase"=$Particles.NormalizedAge
     "[1] Frequency"=&RandomRangeFloat(Minimum=0.15, Maximum=2)
     "[1] Amplitude Min/Max"=<0 1>
    )
   )
   &CurlNoiseForce(Noise Frequency=15, Noise Strength=&RandomRangeFloat(Minimum=60, Maximum=250), Pan Noise=false)
   &InheritSourceMovement(Applied Force Scale=<0.1 0.1 0>, ApplyForce=true, ApplyForceFalloff=false, ApplyPosition=false, Module.LimitForceSpeed=false)
   &SolveForcesAndVelocity(Clamp Velocity=false, Limit Acceleration=false)
   &ScaleColor(
    Scale Alpha=&FloatFromCurve(
     FloatCurve=DataInterface() {propertyValues {Curve {keys [
         {arriveTangent=9.0725, interpMode=RCIM_Cubic, leaveTangent=9.0725, tangentMode=RCTM_User, time=0, value=0}
         {arriveTangent=-2.5138, interpMode=RCIM_Cubic, leaveTangent=-2.5138, tangentMode=RCTM_User, time=0.1408, value=2}
         {arriveTangent=-0.0405, interpMode=RCIM_Cubic, leaveTangent=-0.0405, tangentMode=RCTM_User, time=1, value=0}
        ]}}}
     Scale Curve=0.5
    )
    ScaleA=true
    ScaleRGB=true
   )
  ]
  rendererClasses [&NiagaraSpriteRendererProperties]
  renderers [{rendererClass=&NiagaraSpriteRendererProperties}]
 }
}
summary {
 systemName=NS_Pickaxe_SpaceFeline_Trail
 emitters [
  {emitterName=HoloTrail, rendererClasses [&NiagaraRibbonRendererProperties]}
  {emitterName=FillTrail, rendererClasses [&NiagaraRibbonRendererProperties]}
  {emitterName=Squares_Var0and1_Only, rendererClasses [&NiagaraSpriteRendererProperties]}
  {bEnabled=false, emitterName=TrailParticles_OtherVars, rendererClasses [&NiagaraSpriteRendererProperties]}
 ]
 userVariables [
  User.RibbonIDOffset {type=NiagaraInt32}
  User.SkeletalMesh {
   type=NiagaraDataInterfaceSkeletalMesh
   defaultValue {
    struct=NiagaraExt_VariableValue_DataInterface
    value {dataInterface=&self:NiagaraDataInterfaceSkeletalMesh_0}
   }
  }
  User.VariantID {type=NiagaraFloat, defaultValue=0, description=""}
 ]
}
