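# Physics asset for the SpaceFeline pickaxe, defining collision and joints for its two-handed handle and an attached chain.
# Four capsules shape the handle, which is locked to the root along with the two base chain links.
# The first dynamic link swings freely on one axis like a hinge; later links have limited swing and almost no twist.
# The chain ends in a much larger capsule, suggesting a dangling ornament.
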PhysicsAsset /BRCosmetics/Weapons/FORT_Melee/Pickaxe_SpaceFeline/Meshes/SpaceFeline_Axe_Physics
bodies {
 Melee_TwoHanded_Handle [
  {
   shapeType=Capsule
   length=10
   radius=3
   shapeName=Melee_TwoHanded_Handle_capsule
   center=<2.572 7.4713 -40.9627>
   rotation=<rot 44.7478 -90.9839 -0.0327>
  }
  {
   shapeType=Capsule
   length=13.5073
   radius=4
   shapeName=Melee_TwoHanded_Handle_capsule
   center=<1.7302 4.0481 -31.9328>
   rotation=<rot 7.9876 -6.1238 -52.5321>
  }
  {shapeType=Capsule, length=6, radius=8, shapeName=Melee_TwoHanded_Handle_capsule, center=<0 -1.7611 -13.8508>, rotation=<rot 0 0 178.3963>}
  {
   shapeType=Capsule
   length=4.5
   radius=1.2
   shapeName=Melee_TwoHanded_Handle_capsule
   center=<1.1172 11.431 -35.0408>
   rotation=<rot 0 0 119.9997>
  }
 ]
 b_AR_Root [{shapeType=Capsule, length=0.505, radius=0.505, shapeName=b_AR_Root_capsule, rotation=<rot 0 -90 0>}]
 base_chain [{shapeType=Capsule, length=0.5, radius=0.5, shapeName=base_chain_capsule, rotation=<rot 90 0 0>}]
 base_chain_01 [{shapeType=Capsule, length=1, radius=0.7, shapeName=base_chain_01_capsule, center=<0.5787 0 0>, rotation=<rot 90 0 0>}]
 dyn_chainT_01 [{shapeType=Capsule, length=1, radius=0.7, shapeName=dyn_chainT_01_capsule, center=<1 0 0>, rotation=<rot 90 0 0>}]
 dyn_chain_01 [{shapeType=Capsule, length=1, radius=0.7, shapeName=dyn_chain_01_capsule, center=<-0.5 0 0>, rotation=<rot 90 0 0>}]
 dyn_chain_02 [{shapeType=Capsule, length=1, radius=0.7, shapeName=dyn_chain_02_capsule, center=<1 0 0>, rotation=<rot 90 0 0>}]
 dyn_chain_03 [{shapeType=Capsule, length=1, radius=0.7, shapeName=dyn_chain_03_capsule, center=<1 0 0>, rotation=<rot 90 0 0>}]
 dyn_chain_04 [{shapeType=Capsule, length=1, radius=0.7, shapeName=dyn_chain_04_capsule, center=<1 0 0>, rotation=<rot 90 0 0>}]
 dyn_chain_05 [{shapeType=Capsule, length=1, radius=0.7, shapeName=dyn_chain_05_capsule, center=<1 0 0>, rotation=<rot 90 0 0>}]
 dyn_chain_06 [{shapeType=Capsule, length=1.7621, radius=4.8859, shapeName=dyn_chain_06_capsule, center=<7.2159 1.0633 0>, rotation=<rot 90 0 0>}]
}
constraints [
 {bone1Name=dyn_chain_03, bone2Name=dyn_chain_02, swing1LimitDegrees=40, swing2LimitDegrees=30, twistLimitDegrees=1}
 {bone1Name=dyn_chain_04, bone2Name=dyn_chain_03, swing1LimitDegrees=25, swing2LimitDegrees=35, twistLimitDegrees=1}
 {bone1Name=dyn_chain_05, bone2Name=dyn_chain_04, swing1LimitDegrees=35, swing2LimitDegrees=35, twistLimitDegrees=1}
 {bone1Name=dyn_chain_06, bone2Name=dyn_chain_05, swing1LimitDegrees=35, swing2LimitDegrees=35, twistLimitDegrees=1}
 {
  bone1Name=Melee_TwoHanded_Handle
  bone2Name=b_AR_Root
  swing1LimitDegrees=45
  swing1Motion=Locked
  swing2LimitDegrees=45
  swing2Motion=Locked
  twistLimitDegrees=45
  twistMotion=Locked
 }
 {
  bone1Name=base_chain
  bone2Name=b_AR_Root
  swing1LimitDegrees=0
  swing1Motion=Locked
  swing2LimitDegrees=0
  swing2Motion=Locked
  twistLimitDegrees=0
  twistMotion=Locked
 }
 {bone1Name=dyn_chain_02, bone2Name=dyn_chainT_01, swing1LimitDegrees=20, swing2LimitDegrees=20, twistLimitDegrees=1}
 {
  bone1Name=base_chain_01
  bone2Name=base_chain
  swing1LimitDegrees=45
  swing1Motion=Locked
  swing2LimitDegrees=45
  swing2Motion=Locked
  twistLimitDegrees=45
  twistMotion=Locked
 }
 {bone1Name=dyn_chain_01, bone2Name=base_chain_01, swing1LimitDegrees=45, swing1Motion=Free, swing2LimitDegrees=0, twistLimitDegrees=1}
 {bone1Name=dyn_chainT_01, bone2Name=dyn_chain_01, swing1LimitDegrees=20, swing2LimitDegrees=20, twistLimitDegrees=1}
]
